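# Physics asset for the left-leg dynamic shin pipes on the female medium Armored Engineer player body.
# Inner and outer pipe bones each get a capsule body, paired with a much smaller capsule on a matching MSTR bone.
# One constraint joins each pipe to its MSTR body, forming two independent pairs with identical swing and twist settings.
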PhysicsAsset /BRCosmetics/Characters/Player/Female/Medium/Bodies/F_MED_Armored_Engineer/Meshes/F_MED_Armored_Engineer_Dyn_leg_l
bodies {
 dyn_shin_pipe_in_l [{shapeType=Capsule, length=3.3246, radius=3.1824, center=<0.5 0 0>, rotation=<rot 90 0 0>}]
 dyn_shin_pipe_in_l_MSTR [{shapeType=Capsule, length=0.505, radius=0.505}]
 dyn_shin_pipe_out_l [{shapeType=Capsule, length=3.3246, radius=3.1824, center=<0.5 0 0>, rotation=<rot 90 0 0>}]
 dyn_shin_pipe_out_l_MSTR [{shapeType=Capsule, length=0.505, radius=0.505}]
}
constraints [
 {bone1Name=dyn_shin_pipe_out_l, bone2Name=dyn_shin_pipe_out_l_MSTR, swing1LimitDegrees=90, swing2LimitDegrees=90, twistLimitDegrees=90}
 {bone1Name=dyn_shin_pipe_in_l, bone2Name=dyn_shin_pipe_in_l_MSTR, swing1LimitDegrees=90, swing2LimitDegrees=90, twistLimitDegrees=90}
]
