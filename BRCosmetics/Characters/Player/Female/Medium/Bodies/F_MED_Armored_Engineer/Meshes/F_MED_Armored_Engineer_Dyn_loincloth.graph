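# Physics asset for the dynamic secondary parts of the female medium Armored Engineer body, named for a loincloth but covering left and right gun and holster bones.
# Each of the four bones has one capsule, with larger ones on the guns and tiny ones on the holsters.
# One constraint per side pairs gun with holster, allowing narrow swing on one axis and narrow twist, with the second swing axis locked.
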PhysicsAsset /BRCosmetics/Characters/Player/Female/Medium/Bodies/F_MED_Armored_Engineer/Meshes/F_MED_Armored_Engineer_Dyn_loincloth
bodies {
 dyn_gun_l [{shapeType=Capsule, length=3.3246, radius=3.1824, center=<-0.4958 0 0>, rotation=<rot 90 0 0>}]
 dyn_gun_r [{shapeType=Capsule, length=3.3246, radius=3.1824, center=<-0.4958 0 0>, rotation=<rot 90 0 0>}]
 dyn_holster_l [{shapeType=Capsule, length=0.505, radius=0.505}]
 dyn_holster_r [{shapeType=Capsule, length=0.505, radius=0.505}]
}
constraints [
 {bone1Name=dyn_gun_l, bone2Name=dyn_holster_l, swing1LimitDegrees=10, swing2LimitDegrees=0, swing2Motion=Locked, twistLimitDegrees=10}
 {bone1Name=dyn_gun_r, bone2Name=dyn_holster_r, swing1LimitDegrees=10, swing2LimitDegrees=0, swing2Motion=Locked, twistLimitDegrees=10}
]
