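# Physics asset of cape colliders for the female medium Armored Engineer body, evidently for cape cloth to collide against.
# Capsules cover the pelvis, spine, clavicles, neck and thighs, with a sphere on the head and several overlapping capsules on the pelvis, spine and thighs.
# Arm and calf bodies carry no shapes.
# Constraints link the bones along the skeleton with all swing and twist motion free, so they impose no angular limits.
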PhysicsAsset /BRCosmetics/Characters/Player/Female/Medium/Bodies/F_MED_Armored_Engineer/Meshes/F_MED_Armored_Engineer_CapeColliders
bodies {
 calf_l []
 calf_r []
 clavicle_l [{shapeType=Capsule, length=8.2324, radius=6.1115, center=<9.2486 -1.6905 -0.1052>, rotation=<rot 75.9351 -0.0001 -180>}]
 clavicle_r [{shapeType=Capsule, length=8.2324, radius=6.1115, center=<-9.2486 1.6905 0.1052>, rotation=<rot 75.9351 -0.0001 0>}]
 head [{shapeType=Sphere, radius=17.3125, center=<11.028 -0.174 0>}]
 lowerarm_l []
 lowerarm_r []
 neck_02 [{shapeType=Capsule, length=13.4375, radius=7.3175, center=<-0.2642 -0.3408 0>, rotation=<rot 90 -171.1193 -175.3978>}]
 pelvis [
  {shapeType=Capsule, length=8.4687, radius=13.7263, center=<-0.7734 -1.121 -0.0003>}
  {shapeType=Capsule, length=14.625, radius=19.4375, center=<-17.2793 2.5044 0>, rotation=<rot 90 -0.0396 -0.0396>}
 ]
 spine_01 [{shapeType=Capsule, length=10.9375, radius=12.195, center=<-2.6215 0.1902 -0.0001>}]
 spine_02 [
  {shapeType=Capsule, length=5.125, radius=10.945, center=<-2.6215 0.4993 0.0001>}
  {shapeType=Capsule, length=29.875, radius=11.1013, center=<2.5024 0.4991 1.1546>, rotation=<rot 90 0.0198 0.0198>}
  {shapeType=Capsule, length=29.875, radius=11.1013, center=<2.4704 0.4992 -1.304>, rotation=<rot 90 -0.0049 -0.0049>}
 ]
 spine_03 [{shapeType=Capsule, length=4.8437, radius=10.195, center=<-1.953 0.1364 0.0001>}]
 spine_04 [{shapeType=Capsule, length=2.9375, radius=12.1638, center=<-0.0327 -0.4036 0.0002>}]
 spine_05 [
  {shapeType=Capsule, length=7.9062, radius=11.5388, center=<-3.1632 -0.705 0>}
  {shapeType=Capsule, length=10.0938, radius=11.5938, center=<-0.6475 -0.4153 0>}
 ]
 thigh_l [
  {shapeType=Capsule, length=16.3125, radius=4.0625, center=<-8.5606 -5.0357 -10.2962>, rotation=<rot 84.6098 158.4942 -15.7772>}
  {shapeType=Capsule, length=0.7188, radius=9.25, center=<-0.1844 -5.4054 -8.5462>, rotation=<rot -1.8998 16.4877 102.879>}
 ]
 thigh_r [
  {shapeType=Capsule, length=16.3125, radius=4.0625, center=<8.5606 5.0357 10.2962>, rotation=<rot 84.6098 158.4942 164.2229>}
  {shapeType=Capsule, length=0.7188, radius=9.5, center=<0.1844 5.4054 8.5462>, rotation=<rot -1.8998 16.4877 -77.121>}
 ]
 upperarm_l []
 upperarm_r []
}
constraints [
 {
  bone1Name=spine_03
  bone2Name=spine_02
  swing1LimitDegrees=45
  swing1Motion=Free
  swing2LimitDegrees=45
  swing2Motion=Free
  twistLimitDegrees=45
  twistMotion=Free
 }
 {
  bone1Name=spine_05
  bone2Name=spine_04
  swing1LimitDegrees=45
  swing1Motion=Free
  swing2LimitDegrees=45
  swing2Motion=Free
  twistLimitDegrees=45
  twistMotion=Free
 }
 {
  bone1Name=clavicle_l
  bone2Name=spine_05
  swing1LimitDegrees=45
  swing1Motion=Free
  swing2LimitDegrees=45
  swing2Motion=Free
  twistLimitDegrees=45
  twistMotion=Free
 }
 {
  bone1Name=upperarm_l
  bone2Name=clavicle_l
  swing1LimitDegrees=45
  swing1Motion=Free
  swing2LimitDegrees=45
  swing2Motion=Free
  twistLimitDegrees=45
  twistMotion=Free
 }
 {
  bone1Name=calf_l
  bone2Name=thigh_l
  swing1LimitDegrees=45
  swing1Motion=Free
  swing2LimitDegrees=45
  swing2Motion=Free
  twistLimitDegrees=45
  twistMotion=Free
 }
 {
  bone1Name=head
  bone2Name=neck_02
  swing1LimitDegrees=45
  swing1Motion=Free
  swing2LimitDegrees=45
  swing2Motion=Free
  twistLimitDegrees=45
  twistMotion=Free
 }
 {
  bone1Name=neck_02
  bone2Name=spine_05
  swing1LimitDegrees=45
  swing1Motion=Free
  swing2LimitDegrees=45
  swing2Motion=Free
  twistLimitDegrees=45
  twistMotion=Free
 }
 {
  bone1Name=spine_02
  bone2Name=spine_01
  swing1LimitDegrees=45
  swing1Motion=Free
  swing2LimitDegrees=45
  swing2Motion=Free
  twistLimitDegrees=45
  twistMotion=Free
 }
 {
  bone1Name=spine_01
  bone2Name=pelvis
  swing1LimitDegrees=45
  swing1Motion=Free
  swing2LimitDegrees=45
  swing2Motion=Free
  twistLimitDegrees=45
  twistMotion=Free
 }
 {
  bone1Name=spine_04
  bone2Name=spine_03
  swing1LimitDegrees=45
  swing1Motion=Free
  swing2LimitDegrees=45
  swing2Motion=Free
  twistLimitDegrees=45
  twistMotion=Free
 }
 {
  bone1Name=lowerarm_l
  bone2Name=upperarm_l
  swing1LimitDegrees=45
  swing1Motion=Free
  swing2LimitDegrees=45
  swing2Motion=Free
  twistLimitDegrees=45
  twistMotion=Free
 }
 {
  bone1Name=upperarm_r
  bone2Name=clavicle_r
  swing1LimitDegrees=45
  swing1Motion=Free
  swing2LimitDegrees=45
  swing2Motion=Free
  twistLimitDegrees=45
  twistMotion=Free
 }
 {
  bone1Name=clavicle_r
  bone2Name=spine_05
  swing1LimitDegrees=45
  swing1Motion=Free
  swing2LimitDegrees=45
  swing2Motion=Free
  twistLimitDegrees=45
  twistMotion=Free
 }
 {
  bone1Name=lowerarm_r
  bone2Name=upperarm_r
  swing1LimitDegrees=45
  swing1Motion=Free
  swing2LimitDegrees=45
  swing2Motion=Free
  twistLimitDegrees=45
  twistMotion=Free
 }
 {
  bone1Name=calf_r
  bone2Name=thigh_r
  swing1LimitDegrees=45
  swing1Motion=Free
  swing2LimitDegrees=45
  swing2Motion=Free
  twistLimitDegrees=45
  twistMotion=Free
 }
 {
  bone1Name=thigh_r
  bone2Name=pelvis
  swing1LimitDegrees=45
  swing1Motion=Free
  swing2LimitDegrees=45
  swing2Motion=Free
  twistLimitDegrees=45
  twistMotion=Free
 }
 {
  bone1Name=thigh_l
  bone2Name=pelvis
  swing1LimitDegrees=45
  swing1Motion=Free
  swing2LimitDegrees=45
  swing2Motion=Free
  twistLimitDegrees=45
  twistMotion=Free
 }
]
