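# Physics asset for the Armored Engineer medium female player body, named as the collision set for its cloth simulation.
# The pelvis and three spine bodies each carry an identical offset capsule, while the thigh and calf bodies list no shapes.
# Constraints chain the spine and both legs to the pelvis, with all swing and twist motion free.
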PhysicsAsset /BRCosmetics/Characters/Player/Female/Medium/Bodies/F_MED_Armored_Engineer/Meshes/F_MED_Armored_Engineer_ClothCollision
bodies {
 calf_l []
 calf_r []
 pelvis [{shapeType=Capsule, length=11.7187, radius=10.4763, center=<-5.9663 -1.121 -0.0001>}]
 spine_01 [{shapeType=Capsule, length=11.7187, radius=10.4763, center=<-5.9663 -1.121 -0.0001>}]
 spine_02 [{shapeType=Capsule, length=11.7187, radius=10.4763, center=<-5.9663 -1.121 -0.0001>}]
 spine_03 [{shapeType=Capsule, length=11.7187, radius=10.4763, center=<-5.9663 -1.121 -0.0001>}]
 thigh_l []
 thigh_r []
}
constraints [
 {
  bone1Name=calf_l
  bone2Name=thigh_l
  swing1LimitDegrees=45
  swing1Motion=Free
  swing2LimitDegrees=45
  swing2Motion=Free
  twistLimitDegrees=45
  twistMotion=Free
 }
 {
  bone1Name=thigh_l
  bone2Name=pelvis
  swing1LimitDegrees=45
  swing1Motion=Free
  swing2LimitDegrees=45
  swing2Motion=Free
  twistLimitDegrees=45
  twistMotion=Free
 }
 {
  bone1Name=calf_r
  bone2Name=thigh_r
  swing1LimitDegrees=45
  swing1Motion=Free
  swing2LimitDegrees=45
  swing2Motion=Free
  twistLimitDegrees=45
  twistMotion=Free
 }
 {
  bone1Name=thigh_r
  bone2Name=pelvis
  swing1LimitDegrees=45
  swing1Motion=Free
  swing2LimitDegrees=45
  swing2Motion=Free
  twistLimitDegrees=45
  twistMotion=Free
 }
 {
  bone1Name=spine_01
  bone2Name=pelvis
  swing1LimitDegrees=45
  swing1Motion=Free
  swing2LimitDegrees=45
  swing2Motion=Free
  twistLimitDegrees=45
  twistMotion=Free
 }
 {
  bone1Name=spine_02
  bone2Name=spine_01
  swing1LimitDegrees=45
  swing1Motion=Free
  swing2LimitDegrees=45
  swing2Motion=Free
  twistLimitDegrees=45
  twistMotion=Free
 }
 {
  bone1Name=spine_03
  bone2Name=spine_02
  swing1LimitDegrees=45
  swing1Motion=Free
  swing2LimitDegrees=45
  swing2Motion=Free
  twistLimitDegrees=45
  twistMotion=Free
 }
]
