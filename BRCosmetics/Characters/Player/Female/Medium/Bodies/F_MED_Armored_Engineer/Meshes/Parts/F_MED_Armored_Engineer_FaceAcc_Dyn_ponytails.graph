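# Physics asset for the ponytails face accessory on the female medium Armored Engineer body, defining collision and joints for the left and right dynamic hair tails.
# Each tail is a capsule chain from a small base on the head through segments a to f, with angular limits that favor swing1, tightest at segment a.
# Capsules on the head, neck, upper spine, clavicles and upper arms form the body shapes around the hair, joined by free-rotation constraints.
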PhysicsAsset /BRCosmetics/Characters/Player/Female/Medium/Bodies/F_MED_Armored_Engineer/Meshes/Parts/F_MED_Armored_Engineer_FaceAcc_Dyn_ponytails
bodies {
 clavicle_l [{shapeType=Capsule, length=9.125, radius=7.2188, center=<6.4838 -0.884 -0.9247>, rotation=<rot 83.5144 0.0344 -0.1132>}]
 clavicle_r [{shapeType=Capsule, length=9.125, radius=7.2188, center=<-6.4838 0.884 0.9247>, rotation=<rot 83.5132 0.0344 179.8868>}]
 dyn_hair_tail_a_base_l [{shapeType=Capsule, length=0.505, radius=0.505, rotation=<rot 0 -90 0>}]
 dyn_hair_tail_a_base_r [{shapeType=Capsule, length=0.505, radius=0.505, rotation=<rot 0 -90 0>}]
 dyn_hair_tail_a_l [{shapeType=Capsule, length=2.9688, radius=2.1875, center=<2.2232 0.1549 0.4888>, rotation=<rot 77.9135 -166.6406 179.9999>}]
 dyn_hair_tail_a_r [{shapeType=Capsule, length=2.9688, radius=2.1875, center=<-2.2232 -0.1549 -0.4888>, rotation=<rot 77.9135 -166.6406 -0.0001>}]
 dyn_hair_tail_b_l [{shapeType=Capsule, length=2.9688, radius=2.1875, center=<2.451 0 0>, rotation=<rot 90 0 0>}]
 dyn_hair_tail_b_r [{shapeType=Capsule, length=2.9688, radius=2.1875, center=<-2.451 0 0>, rotation=<rot 90 0 180>}]
 dyn_hair_tail_c_l [{shapeType=Capsule, length=3.25, radius=1.9062, center=<2.451 0 0>, rotation=<rot 90 0 0>}]
 dyn_hair_tail_c_r [{shapeType=Capsule, length=3.25, radius=1.9062, center=<-2.451 0 0>, rotation=<rot 90 0 180>}]
 dyn_hair_tail_d_l [{shapeType=Capsule, length=3.375, radius=1.7812, center=<2.451 0 0>, rotation=<rot 90 0 0>}]
 dyn_hair_tail_d_r [{shapeType=Capsule, length=3.375, radius=1.7812, center=<-2.451 0 0>, rotation=<rot 90 0 180>}]
 dyn_hair_tail_e_l [{shapeType=Capsule, length=3.375, radius=1.7812, center=<2.451 0 0>, rotation=<rot 90 0 0>}]
 dyn_hair_tail_e_r [{shapeType=Capsule, length=3.375, radius=1.7812, center=<-2.451 0 0>, rotation=<rot 90 0 180>}]
 dyn_hair_tail_f_l [{shapeType=Capsule, length=4.5, radius=1.9688, center=<4.2132 0 -0.0001>, rotation=<rot 90 0 0>}]
 dyn_hair_tail_f_r [{shapeType=Capsule, length=4.5, radius=1.9688, center=<-4.2132 0 0.0001>, rotation=<rot 90 0 180>}]
 head [{shapeType=Capsule, length=8.25, radius=6, center=<3.1816 1.3754 0.0001>, rotation=<rot 90 0 0>}]
 neck_01 [{shapeType=Capsule, length=6.0938, radius=5.1562, center=<0.8316 -0.1459 -0.1493>, rotation=<rot 90 -10.6197 -13.3382>}]
 spine_04 [{shapeType=Capsule, length=4.7812, radius=10.5312, center=<0.7443 4.6385 0.0001>}]
 spine_05 [
  {shapeType=Capsule, length=8.125, radius=10.4062, center=<0.5852 -1.3864 0>}
  {shapeType=Capsule, length=4.5938, radius=2.8125, center=<10.1688 2.8882 -4.7332>, rotation=<rot 0 25.6498 89.9997>}
  {shapeType=Capsule, length=4.5938, radius=2.8125, center=<10.1376 2.8872 5.0196>, rotation=<rot 0 25.6467 89.9997>}
 ]
 upperarm_l [{shapeType=Capsule, length=15.8438, radius=5.6562, center=<11.0156 -0.6175 -0.4439>, rotation=<rot 89.5554 180 -179.042>}]
 upperarm_r [{shapeType=Capsule, length=15.8438, radius=5.6562, center=<-11.0156 0.6175 0.4439>, rotation=<rot 89.5532 180 0.958>}]
}
constraints [
 {
  bone1Name=head
  bone2Name=neck_01
  swing1LimitDegrees=45
  swing1Motion=Free
  swing2LimitDegrees=45
  swing2Motion=Free
  twistLimitDegrees=45
  twistMotion=Free
 }
 {
  bone1Name=clavicle_l
  bone2Name=spine_05
  swing1LimitDegrees=45
  swing1Motion=Free
  swing2LimitDegrees=45
  swing2Motion=Free
  twistLimitDegrees=45
  twistMotion=Free
 }
 {
  bone1Name=spine_05
  bone2Name=spine_04
  swing1LimitDegrees=45
  swing1Motion=Free
  swing2LimitDegrees=45
  swing2Motion=Free
  twistLimitDegrees=45
  twistMotion=Free
 }
 {
  bone1Name=neck_01
  bone2Name=spine_05
  swing1LimitDegrees=45
  swing1Motion=Free
  swing2LimitDegrees=45
  swing2Motion=Free
  twistLimitDegrees=45
  twistMotion=Free
 }
 {
  bone1Name=dyn_hair_tail_a_base_l
  bone2Name=head
  swing1LimitDegrees=45
  swing1Motion=Free
  swing2LimitDegrees=45
  swing2Motion=Free
  twistLimitDegrees=45
  twistMotion=Free
 }
 {
  bone1Name=dyn_hair_tail_a_base_r
  bone2Name=head
  swing1LimitDegrees=45
  swing1Motion=Free
  swing2LimitDegrees=45
  swing2Motion=Free
  twistLimitDegrees=45
  twistMotion=Free
 }
 {bone1Name=dyn_hair_tail_a_l, bone2Name=dyn_hair_tail_a_base_l, swing1LimitDegrees=45, swing2LimitDegrees=5, twistLimitDegrees=5}
 {bone1Name=dyn_hair_tail_b_l, bone2Name=dyn_hair_tail_a_l, swing1LimitDegrees=45, swing2LimitDegrees=15, twistLimitDegrees=5}
 {bone1Name=dyn_hair_tail_c_l, bone2Name=dyn_hair_tail_b_l, swing1LimitDegrees=45, swing2LimitDegrees=15, twistLimitDegrees=5}
 {bone1Name=dyn_hair_tail_d_l, bone2Name=dyn_hair_tail_c_l, swing1LimitDegrees=45, swing2LimitDegrees=15, twistLimitDegrees=5}
 {bone1Name=dyn_hair_tail_e_l, bone2Name=dyn_hair_tail_d_l, swing1LimitDegrees=45, swing2LimitDegrees=15, twistLimitDegrees=5}
 {bone1Name=dyn_hair_tail_f_l, bone2Name=dyn_hair_tail_e_l, swing1LimitDegrees=45, swing2LimitDegrees=15, twistLimitDegrees=5}
 {
  bone1Name=upperarm_l
  bone2Name=clavicle_l
  swing1LimitDegrees=45
  swing1Motion=Free
  swing2LimitDegrees=45
  swing2Motion=Free
  twistLimitDegrees=45
  twistMotion=Free
 }
 {bone1Name=dyn_hair_tail_a_r, bone2Name=dyn_hair_tail_a_base_r, swing1LimitDegrees=45, swing2LimitDegrees=5, twistLimitDegrees=5}
 {bone1Name=dyn_hair_tail_b_r, bone2Name=dyn_hair_tail_a_r, swing1LimitDegrees=45, swing2LimitDegrees=15, twistLimitDegrees=5}
 {bone1Name=dyn_hair_tail_c_r, bone2Name=dyn_hair_tail_b_r, swing1LimitDegrees=45, swing2LimitDegrees=15, twistLimitDegrees=5}
 {bone1Name=dyn_hair_tail_d_r, bone2Name=dyn_hair_tail_c_r, swing1LimitDegrees=45, swing2LimitDegrees=15, twistLimitDegrees=5}
 {bone1Name=dyn_hair_tail_e_r, bone2Name=dyn_hair_tail_d_r, swing1LimitDegrees=45, swing2LimitDegrees=15, twistLimitDegrees=5}
 {bone1Name=dyn_hair_tail_f_r, bone2Name=dyn_hair_tail_e_r, swing1LimitDegrees=45, swing2LimitDegrees=15, twistLimitDegrees=5}
 {
  bone1Name=clavicle_r
  bone2Name=spine_05
  swing1LimitDegrees=45
  swing1Motion=Free
  swing2LimitDegrees=45
  swing2Motion=Free
  twistLimitDegrees=45
  twistMotion=Free
 }
 {
  bone1Name=upperarm_r
  bone2Name=clavicle_r
  swing1LimitDegrees=45
  swing1Motion=Free
  swing2LimitDegrees=45
  swing2Motion=Free
  twistLimitDegrees=45
  twistMotion=Free
 }
]
