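# Post-body anim blueprint for the female Armored Engineer's cosmetic part, handling nose ring and hair on a pose copied from the attached parent mesh.
# A constraint blends the nose ring between nose bridge and head.
# Head rotation relative to spine_05 sets headUp and headTwist, whose range-mapped alphas drive additive pigtail-base offsets.
# Two RigidBody nodes then simulate ponytails and hair in neck_01 space with separate physics assets.
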AnimBlueprint /BRCosmetics/Characters/Player/Female/Medium/Bodies/F_MED_Armored_Engineer/Meshes/Parts/F_MED_Armored_Engineer_AnimBP
target_skeleton=&/F_MED_Armored_Engineer_Skeleton
graphs [
 AnimGraph {nodes [
   n1 = "Constraint - C_nose_ring_MSTR (PP)"!(
    ComponentPose=Local To Component(LocalPose=Copy Pose From Mesh() {bUseAttachedParent=true})
    Alpha=1
    ConstraintWeights_0=0.3
    ConstraintWeights_1:float=0.7
   ) -> (Pose) {
    alphaBoolBlend {blendOption=Linear}
    boneToModify {boneName=C_nose_ring_MSTR}
    constraintSetup [{targetBone {boneName=C_nose_bridge}}, {targetBone {boneName=head}}]
    constraintWeights [1,1]
   }
   n2 = "Transform (Modify) Bone - Bone: dyn_hair_tail_a_base_l"(ComponentPose=@n1, Alpha=0, Translation="0,0,-1", Rotation="-10,0,0", Scale="1,1,1") {
    rotationMode=BMM_Additive
    rotationSpace=BCS_BoneSpace
    translationMode=BMM_Additive
    translationSpace=BCS_BoneSpace
    alphaBoolBlend {blendOption=Linear}
    alphaScaleBiasClamp {inRange {max=0, min=-50}, outRange {max=0, min=1}}
    boneToModify {boneName=dyn_hair_tail_a_base_l}
    rotation {pitch=-10}
    translation {z=-1}
   }
   n3 = "Transform (Modify) Bone - Bone: dyn_hair_tail_a_base_r"(ComponentPose=@n2, Alpha=0, Translation="0,0,0", Rotation="-10,0,0", Scale="1,1,1") {
    rotationMode=BMM_Additive
    rotationSpace=BCS_BoneSpace
    translationMode=BMM_Additive
    translationSpace=BCS_BoneSpace
    alphaBoolBlend {blendOption=Linear}
    alphaScaleBiasClamp {inRange {max=0, min=-50}, outRange {max=0, min=1}}
    boneToModify {boneName=dyn_hair_tail_a_base_r}
    rotation {pitch=-10}
   }
   n4 = Get HeadUp() -> (HeadUp:double)
   n5 = "Transform (Modify) Bone - Bone: dyn_hair_tail_a_base_l"(ComponentPose=@n3, Alpha=@n4, Translation="0,0,0", Rotation="-5,-45,0", Scale="1,1,1") {
    rotationMode=BMM_Additive
    rotationSpace=BCS_BoneSpace
    translationMode=BMM_Additive
    translationSpace=BCS_BoneSpace
    alphaBoolBlend {blendOption=Linear}
    alphaScaleBiasClamp {bMapRange=true, inRange {max=0, min=-50}, outRange {max=0, min=1}}
    boneToModify {boneName=dyn_hair_tail_a_base_l}
    rotation {pitch=-5, yaw=-45}
   }
   n6 = "Transform (Modify) Bone - Bone: dyn_hair_tail_a_base_l"(ComponentPose=@n5, Alpha=@n4, Translation="0,0,0", Rotation="0,15,0", Scale="1,1,1") {
    rotationMode=BMM_Additive
    rotationSpace=BCS_BoneSpace
    translationMode=BMM_Additive
    translationSpace=BCS_BoneSpace
    alphaBoolBlend {blendOption=Linear}
    alphaScaleBiasClamp {bMapRange=true, inRange {max=15}}
    boneToModify {boneName=dyn_hair_tail_a_base_l}
    rotation {yaw=15}
   }
   n7 = Get HeadUp()
   n8 = "Transform (Modify) Bone - Bone: dyn_hair_tail_a_base_r"(ComponentPose=@n6, Alpha=@n7, Translation="0,0,0", Rotation="-5,-45,0", Scale="1,1,1") {
    rotationMode=BMM_Additive
    rotationSpace=BCS_BoneSpace
    translationMode=BMM_Additive
    translationSpace=BCS_BoneSpace
    alphaBoolBlend {blendOption=Linear}
    alphaScaleBiasClamp {bMapRange=true, inRange {max=0, min=-50}, outRange {max=0, min=1}}
    boneToModify {boneName=dyn_hair_tail_a_base_r}
    rotation {pitch=-5, yaw=-45}
   }
   n9 = "Transform (Modify) Bone - Bone: dyn_hair_tail_a_base_r"(ComponentPose=@n8, Alpha=@n7, Translation="0,0,0", Rotation="0,15,0", Scale="1,1,1") {
    rotationMode=BMM_Additive
    rotationSpace=BCS_BoneSpace
    translationMode=BMM_Additive
    translationSpace=BCS_BoneSpace
    alphaBoolBlend {blendOption=Linear}
    alphaScaleBiasClamp {bMapRange=true, inRange {max=15}}
    boneToModify {boneName=dyn_hair_tail_a_base_r}
    rotation {yaw=15}
   }
   n10 = Get HeadTwist() -> (HeadTwist:double)
   n11 = "Transform (Modify) Bone - Bone: dyn_hair_tail_a_base_l"(ComponentPose=@n9, Alpha=@n10, Translation="0,2,0", Rotation="-20,0,0", Scale="1,1,1") {
    rotationMode=BMM_Additive
    rotationSpace=BCS_BoneSpace
    translationMode=BMM_Additive
    translationSpace=BCS_BoneSpace
    alphaBoolBlend {blendOption=Linear}
    alphaScaleBiasClamp {bMapRange=true, inRange {max=40}}
    boneToModify {boneName=dyn_hair_tail_a_base_l}
    rotation {pitch=-20}
    translation {y=2}
   }
   n12 = "Transform (Modify) Bone - Bone: dyn_hair_tail_a_base_l"(ComponentPose=@n11, Alpha=@n10, Translation="0,-2.5,0", Rotation="20,-30,0", Scale="1,1,1") {
    rotationMode=BMM_Additive
    rotationSpace=BCS_BoneSpace
    translationMode=BMM_Additive
    translationSpace=BCS_BoneSpace
    alphaBoolBlend {blendOption=Linear}
    alphaScaleBiasClamp {bMapRange=true, inRange {max=0, min=-40}, outRange {max=0, min=1}}
    boneToModify {boneName=dyn_hair_tail_a_base_l}
    rotation {pitch=20, yaw=-30}
    translation {y=-2.5}
   }
   n13 = Get HeadTwist()
   n14 = "Transform (Modify) Bone - Bone: dyn_hair_tail_a_base_r"(ComponentPose=@n12, Alpha=@n13, Translation="0,2.5,0", Rotation="20,-25,0", Scale="1,1,1") {
    rotationMode=BMM_Additive
    rotationSpace=BCS_BoneSpace
    translationMode=BMM_Additive
    translationSpace=BCS_BoneSpace
    alphaBoolBlend {blendOption=Linear}
    alphaScaleBiasClamp {bMapRange=true, inRange {max=40}}
    boneToModify {boneName=dyn_hair_tail_a_base_r}
    rotation {pitch=20, yaw=-25}
    translation {y=2.5}
   }
   n15 = "Transform (Modify) Bone - Bone: dyn_hair_tail_a_base_r"(ComponentPose=@n14, Alpha=@n13, Translation="0,-3,0", Rotation="-20,0,0", Scale="1,1,1") {
    rotationMode=BMM_Additive
    rotationSpace=BCS_BoneSpace
    translationMode=BMM_Additive
    translationSpace=BCS_BoneSpace
    alphaBoolBlend {blendOption=Linear}
    alphaScaleBiasClamp {bMapRange=true, inRange {max=0, min=-40}, outRange {max=0, min=1}}
    boneToModify {boneName=dyn_hair_tail_a_base_r}
    rotation {pitch=-20}
    translation {y=-3}
   }
   n16 = RigidBody!(ComponentPose=@n15, Alpha, OverrideWorldGravity:Vector="0,0,-980", ExternalForce) -> (Pose) {
    bAllowPoseScalePropagation=false
    bClampLinearTranslationLimitToRefPose=true
    bOverrideWorldGravity=true
    overridePhysicsAsset=&/F_MED_Armored_Engineer_FaceAcc_Dyn_ponytails
    simulationSpace=BaseBoneSpace
    alphaBoolBlend {blendOption=Linear}
    baseBoneRef {boneName=neck_01}
    componentAppliedLinearAccClamp=<3000 3000 3000>
    componentLinearAccScale {z=0.1}
    componentLinearVelScale {z=0.1}
    overrideWorldGravity {z=-980}
    simSpaceSettings {
     maxAngularAcceleration=1000
     maxAngularVelocity=3
     maxLinearAcceleration=700
     maxLinearVelocity=500
     velocityScaleZ=0
     worldAlpha=0.25
     externalLinearDragV=<0.2 0.2>
    }
   }
   n17 = RigidBody(ComponentPose=@n16, OverrideWorldGravity="0,0,-980") {
    bClampLinearTranslationLimitToRefPose=true
    bOverrideWorldGravity=true
    overridePhysicsAsset=&/F_MED_Armored_Engineer_FaceAcc_Dyn_hair
    simulationSpace=BaseBoneSpace
    alphaBoolBlend {blendOption=Linear}
    baseBoneRef {boneName=neck_01}
    componentAppliedLinearAccClamp=<3000 3000 3000>
    componentLinearAccScale {z=0.1}
    componentLinearVelScale {z=0.1}
    overrideWorldGravity {z=-980}
    simSpaceSettings {
     maxAngularAcceleration=1000
     maxAngularVelocity=3
     maxLinearAcceleration=700
     maxLinearVelocity=500
     velocityScaleZ=0
     worldAlpha=0.25
     externalLinearDragV=<0.2 0.2>
    }
   }
   Output Pose(Result=Component To Local(ComponentPose=@n17)) {layerGroup=None}
  ]}
 EventGraph {nodes [
   Event BlueprintUpdateAnimation()
   n1 = GetOwnerSkeletalMeshComponent()
   n2 = BreakRotator(InRot=BreakTransform(InTransform=GetDeltaTransformFromRefPose(self=@n1, BoneName=head, BaseName=spine_05)).Rotation)
   > Set headTwist(headTwist:double=@n2) -> (Output_Get:double)
   > Set headUp(headUp:double=@n2.Yaw) -> (Output_Get:double)
  ]}
]
variables [
 headTwist {type=number}
 headUp {type=number}
]
